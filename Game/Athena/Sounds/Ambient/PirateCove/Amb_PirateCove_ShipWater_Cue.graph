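# Sound cue for the looping ambient water sound around the ship at Pirate Cove, in the Athena outdoor ambient sound class.
# A single looping Wave Player plays the dock water-edge wave directly into the cue output, with no randomization or layering.
# The cue overrides attenuation inline, adding distance falloff, occlusion with low-pass filtering on a custom trace channel, a manual reverb send and stereo spread.
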SoundCue /Game/Athena/Sounds/Ambient/PirateCove/Amb_PirateCove_ShipWater_Cue
root=WavePlayer_0
nodes [{id=WavePlayer_0, properties {bLooping=true, soundWaveAssetPtr=&/Amb_PirateCove_WaterEdge_Dock_01}}]
properties {
 bOverrideAttenuation=true
 maxDistance=3000
 soundClassObject=&Fort_Ambient_Outdoors_Athena
 volumeMultiplier=0.802
 attenuationOverrides {
  bEnableFocusInterpolation=true
  bEnableOcclusion=true
  falloffDistance=3000
  focusAttackInterpSpeed=3
  focusReleaseInterpSpeed=3
  manualReverbSendLevel=10
  nonFocusVolumeAttenuation=0.5
  occlusionLowPassFilterFrequency=3000
  occlusionTraceChannel=ECC_GameTraceChannel11
  reverbDistanceMax=2000
  reverbSendMethod=Manual
  reverbWetLevelMax=5
  stereoSpread=1024
  attenuationShapeExtents {x=0}
 }
}
