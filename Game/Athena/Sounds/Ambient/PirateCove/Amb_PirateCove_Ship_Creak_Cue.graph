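# Ambient sound cue for the Pirate Cove ship that plays intermittent wood creaks in Athena.
# It loops a random delay of several seconds into an equally weighted random pick of eight reused house-interior creak waves, passed through a modulator.
# It plays in the Athena outdoor ambient sound class with overridden attenuation, adding occlusion, non-focus ducking and manual reverb send over a limited falloff range.
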SoundCue /Game/Athena/Sounds/Ambient/PirateCove/Amb_PirateCove_Ship_Creak_Cue
root=Looping_0
nodes [
 {id=Delay_0, children [Random_3], properties {delayMax=20, delayMin=5}}
 {id=Looping_0, children [Modulator_0]}
 {id=Modulator_0, children [Delay_0]}
 {
  id=Random_3
  children [WavePlayer_27, WavePlayer_28, WavePlayer_29, WavePlayer_30, WavePlayer_31, WavePlayer_32, WavePlayer_33, WavePlayer_34]
  properties {weights [1,1,1,1,1,1,1,1]}
 }
 {id=WavePlayer_27, properties {soundWaveAssetPtr=/Game/Sounds/Fort_Ambient_NonLoop/House_Interior/house_wood_bigger_creak_07}}
 {id=WavePlayer_28, properties {soundWaveAssetPtr=/Game/Sounds/Fort_Ambient_NonLoop/House_Interior/house_wood_bigger_creak_08}}
 {id=WavePlayer_29, properties {soundWaveAssetPtr=/Game/Sounds/Fort_Ambient_NonLoop/House_Interior/house_wood_bigger_creak_01}}
 {id=WavePlayer_30, properties {soundWaveAssetPtr=/Game/Sounds/Fort_Ambient_NonLoop/House_Interior/house_wood_bigger_creak_02}}
 {id=WavePlayer_31, properties {soundWaveAssetPtr=/Game/Sounds/Fort_Ambient_NonLoop/House_Interior/house_wood_bigger_creak_03}}
 {id=WavePlayer_32, properties {soundWaveAssetPtr=/Game/Sounds/Fort_Ambient_NonLoop/House_Interior/house_wood_bigger_creak_04}}
 {id=WavePlayer_33, properties {soundWaveAssetPtr=/Game/Sounds/Fort_Ambient_NonLoop/House_Interior/house_wood_bigger_creak_05}}
 {id=WavePlayer_34, properties {soundWaveAssetPtr=/Game/Sounds/Fort_Ambient_NonLoop/House_Interior/house_wood_bigger_creak_06}}
]
properties {
 bOverrideAttenuation=true
 maxDistance=2000
 soundClassObject=&Fort_Ambient_Outdoors_Athena
 volumeMultiplier=0.752
 attenuationOverrides {
  bEnableFocusInterpolation=true
  bEnableOcclusion=true
  falloffDistance=2000
  focusAttackInterpSpeed=3
  focusReleaseInterpSpeed=3
  manualReverbSendLevel=10
  nonFocusVolumeAttenuation=0.5
  occlusionInterpolationTime=1
  occlusionLowPassFilterFrequency=5000
  occlusionTraceChannel=ECC_GameTraceChannel11
  occlusionVolumeAttenuation=0.8
  reverbDistanceMax=2000
  reverbSendMethod=Manual
  reverbWetLevelMax=5
  attenuationShapeExtents {x=0}
 }
}
